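# Physics asset for the F_MED_VibrantShell_Chain female player body cosmetic, covering its dangling boba drink and bow ribbon tails.
# Angle-limited constraints form a three-link drink chain from its base and two-link left and right ribbon chains from MSTR_bow, with asymmetric ribbon limits.
# Capsules on the pelvis and right leg only provide body shapes, and a huge root capsule sits below the origin, apparently as a floor.
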PhysicsAsset /BRCosmetics/Characters/Player/Female/Medium/Bodies/F_MED_VibrantShell_Chain/Meshes/F_MED_VibrantShell_Chain_Physics
bodies {
 MSTR_bow [{shapeType=Capsule, length=0.505, radius=0.505, shapeName=MSTR_bow_capsule, rotation=<rot 0 -90 0>}]
 MSTR_skirt [{shapeType=Capsule, length=0.505, radius=0.505, shapeName=MSTR_skirt_capsule, rotation=<rot 0 -90 0>}]
 calf_r [{
   shapeType=Capsule
   length=37.338
   radius=5.5145
   shapeName=calf_r_capsule
   center=<21.9386 2.0995 0.2031>
   rotation=<rot 0.1548 86.072 -88.2302>
  }]
 dyn_boba_drink_1 [{
   shapeType=Capsule
   length=4.0493
   radius=1.8655
   shapeName=dyn_boba_drink_1_capsule
   center=<2.9608 -0.048 -0.2042>
   rotation=<rot 0 90.7953 -78.3879>
  }]
 dyn_boba_drink_2 [{
   shapeType=Capsule
   length=3.965
   radius=3.6676
   shapeName=dyn_boba_drink_2_capsule
   center=<3.2819 -0.456 0.1283>
   rotation=<rot 24.2945 -166.1607 -90>
  }]
 dyn_boba_drink_3 [{
   shapeType=Capsule
   length=8.649
   radius=4.9276
   shapeName=dyn_boba_drink_3_capsule
   center=<5.2465 -0.014 0.0358>
   rotation=<rot 0 90.3869 -89.3941>
  }]
 dyn_boba_drink_base [{
   shapeType=Capsule
   length=0
   radius=1.5374
   shapeName=dyn_boba_drink_base_capsule
   center=<-0.0927 -0.014 -0.0612>
   rotation=<rot -0.9851 2.3863 4.1487>
  }]
 dyn_ribbon_tail_01_l [{
   shapeType=Capsule
   length=7.9219
   radius=3.37
   shapeName=dyn_ribbon_tail_01_l_capsule
   center=<7.1246 0 -1.5571>
   rotation=<rot 0 88.7451 -79.6247>
  }]
 dyn_ribbon_tail_01_r [{
   shapeType=Capsule
   length=8.9669
   radius=2.7465
   shapeName=dyn_ribbon_tail_01_r_capsule
   center=<7.895 0.0354 -1.2315>
   rotation=<rot 0.5602 79.2974 -86.7589>
  }]
 dyn_ribbon_tail_02_l [{
   shapeType=Capsule
   length=10.9902
   radius=3.7161
   shapeName=dyn_ribbon_tail_02_l_capsule
   center=<8.0061 0.1071 -1.424>
   rotation=<rot -0.0333 97.1993 -89.5001>
  }]
 dyn_ribbon_tail_02_r [{
   shapeType=Capsule
   length=11.4786
   radius=2.5344
   shapeName=dyn_ribbon_tail_02_r_capsule
   center=<7.3719 0.3894 1.4488>
   rotation=<rot 83.0149 179.9999 179.9998>
  }]
 foot_r [{
   shapeType=Capsule
   length=12.7112
   radius=6.2382
   shapeName=foot_r_capsule
   center=<3.9669 -7.0951 -0.1711>
   rotation=<rot 0 8.2869 -89.4704>
  }]
 pelvis [{shapeType=Capsule, length=13.414, radius=8.6182, shapeName=pelvis_capsule, center=<-4.4857 -3.1415 -0.002>, rotation=<rot 0 87.4621 0.2346>}]
 root [{shapeType=Capsule, length=5.25, radius=203, shapeName=root_capsule, center=<0 0 -205.096>, rotation=<rot 0 -90 0>}]
 thigh_r [{
   shapeType=Capsule
   length=30.8418
   radius=7.9208
   shapeName=thigh_r_capsule
   center=<18.5658 0.555 -0.5903>
   rotation=<rot 0.0589 -95.9696 -89.4365>
  }]
}
constraints [
 {bone1Name=dyn_boba_drink_1, bone2Name=dyn_boba_drink_base, swing1LimitDegrees=20, swing2LimitDegrees=20, twistLimitDegrees=5}
 {bone1Name=dyn_boba_drink_2, bone2Name=dyn_boba_drink_1, swing1LimitDegrees=25, swing2LimitDegrees=25, twistLimitDegrees=5}
 {bone1Name=dyn_boba_drink_3, bone2Name=dyn_boba_drink_2, swing1LimitDegrees=30, swing2LimitDegrees=30, twistLimitDegrees=10}
 {bone1Name=dyn_ribbon_tail_01_l, bone2Name=MSTR_bow, swing1LimitDegrees=30, swing2LimitDegrees=20, twistLimitDegrees=5}
 {bone1Name=dyn_ribbon_tail_02_l, bone2Name=dyn_ribbon_tail_01_l, swing1LimitDegrees=25, swing2LimitDegrees=35, twistLimitDegrees=5}
 {bone1Name=dyn_ribbon_tail_01_r, bone2Name=MSTR_bow, swing1LimitDegrees=50, swing2LimitDegrees=50, twistLimitDegrees=38}
 {bone1Name=dyn_ribbon_tail_02_r, bone2Name=dyn_ribbon_tail_01_r, swing1LimitDegrees=45, swing2LimitDegrees=75, twistLimitDegrees=5}
]
