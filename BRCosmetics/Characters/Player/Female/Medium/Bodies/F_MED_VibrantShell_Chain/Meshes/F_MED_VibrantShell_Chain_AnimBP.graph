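# Animation Blueprint for the VibrantShell_Chain female body cosmetic, adding secondary motion to its clothing, jewelry and pendants.
# It copies the attached parent mesh's pose into a Control Rig fed by crouch and movement values, then caches it.
# Bone offsets, a pelvis-space RigidBody and per-bone AnimDynamics, some with spherical limits, build on that pose.
# Weighted multi-bone constraints position pendant roots, including a final pass on the cached 'Adjustments' pose before output.
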AnimBlueprint /BRCosmetics/Characters/Player/Female/Medium/Bodies/F_MED_VibrantShell_Chain/Meshes/F_MED_VibrantShell_Chain_AnimBP
target_skeleton=&/F_MED_VibrantShell_Chain_Skeleton
graphs [
 AnimGraph {nodes [
   Anim Dynamics - Bone: dyn_pendant_star_front_02(Alpha=1, ExternalForce="0,0,0") {
    angularDampingOverride=0.9999
    angularSpringConstant=230
    bAngularSpring=true
    bOverrideAngularDamping=true
    bUseGravityOverride=true
    bUseSphericalLimits=true
    gravityScale=0.35
    simulationSpace=RootRelative
    boundBone {boneName=dyn_pendant_star_front_02}
    componentLinearAccScale=<1 1 1>
    componentLinearVelScale=<1 1 1>
    gravityOverride {z=-980}
    physicsBodyDefinitions [{
      collisionType=CustomSphere
      sphereCollisionRadius=5
      boundBone {boneName=dyn_pendant_star_front_02}
      boxExtents=<19 20 20>
      constraintSetup {angularLimitsMax=<15 3 20>, angularLimitsMin=<-15 -15 -20>, angularTarget=<1 0 0>}
      localJointOffset=<4 0 0>
     }]
    simSpaceSettings {maxAngularAcceleration=35, maxAngularVelocity=7, simSpaceAngularAlpha=1}
    sphericalLimits [
     {limitRadius=6, drivingBone {boneName=pelvis}, sphereLocalOffset=<-15 0 0>}
     {limitRadius=3, drivingBone {boneName=thigh_twist_01_l}, sphereLocalOffset=<7 0 2>}
    ]
   }
   "Transform (Modify) Bone - Bone: dyn_pendant_star_front_02"(Alpha=1, Translation="0,0,0", Rotation="22.116,8.694,1.829", Scale="1,1,1") {
    rotationMode=BMM_Additive
    rotationSpace=BCS_ParentBoneSpace
    translationMode=BMM_Additive
    translationSpace=BCS_ParentBoneSpace
    alphaBoolBlend {blendInTime=0.2, blendOutTime=0.4}
    alphaScaleBiasClamp {bInterpResult=true}
    boneToModify {boneName=dyn_pendant_star_front_02}
    rotation=<rot 22.116 8.694 1.829>
   }
   n1 = "Constraint - MSTR_pendant_fox_back (PPPPP)"!(
    ComponentPose=Local To Component(LocalPose=Use cached pose 'Adjustments'())
    Alpha=1
    ConstraintWeights_0=0.3253
    ConstraintWeights_1:float=0.3125
    ConstraintWeights_2:float=0.2354
    ConstraintWeights_3:float=0.0721
    ConstraintWeights_4:float=0.0547
   ) -> (Pose) {
    boneToModify {boneName=MSTR_pendant_fox_back}
    constraintSetup [
     {targetBone {boneName=thigh_twist_01_l}}
     {targetBone {boneName=deform_glute_l}}
     {targetBone {boneName=pelvis}}
     {targetBone {boneName=thigh_l}}
     {targetBone {boneName=dyn_chain_down}}
    ]
    constraintWeights [1,1,1,1,1]
   }
   n2 = "Constraint - MSTR_pendant_star_front (PPPPPPPP)"!(
    ComponentPose=@n1
    Alpha=1
    ConstraintWeights_0=0.3267
    ConstraintWeights_1:float=0.2
    ConstraintWeights_2:float=0.1351
    ConstraintWeights_3:float=0.1235
    ConstraintWeights_4:float=0.1179
    ConstraintWeights_5:float=0.0607
    ConstraintWeights_6:float=0.0337
    ConstraintWeights_7:float=0.0025
   ) -> (Pose) {
    boneToModify {boneName=MSTR_pendant_star_front}
    constraintSetup [
     {targetBone {boneName=thigh_l}}
     {targetBone {boneName=deform_groin_l}}
     {targetBone {boneName=thigh_twist_01_l}}
     {targetBone {boneName=pelvis}}
     {targetBone {boneName=deform_groin_base_l}}
     {targetBone {boneName=correction_chain_down_m}}
     {targetBone {boneName=thigh_twist_01_r}}
     {targetBone {boneName=deform_groin_r}}
    ]
    constraintWeights [1,1,1,1,1,1,1,1]
   }
   n3 = "Constraint - MSTR_pendant_bunny_front (PPPPPPPP)"!(
    ComponentPose=@n2
    Alpha=1
    ConstraintWeights_0=0.2442
    ConstraintWeights_1:float=0.2338
    ConstraintWeights_2:float=0.1949
    ConstraintWeights_3:float=0.1362
    ConstraintWeights_4:float=0.061
    ConstraintWeights_5:float=0.0606
    ConstraintWeights_6:float=0.0409
    ConstraintWeights_7:float=0.0283
   ) -> (Pose) {
    boneToModify {boneName=MSTR_pendant_bunny_front}
    constraintSetup [
     {targetBone {boneName=thigh_l}}
     {targetBone {boneName=correction_chain_down_m}}
     {targetBone {boneName=pelvis}}
     {targetBone {boneName=deform_groin_r}}
     {targetBone {boneName=thigh_twist_01_l}}
     {targetBone {boneName=thigh_twist_01_r}}
     {targetBone {boneName=correction_chain_up_m}}
     {targetBone {boneName=deform_groin_base_l}}
    ]
    constraintWeights [1,1,1,1,1,1,1,1]
   }
   n4 = "Constraint - MSTR_pendant_circle_front (PPPPPP)"!(
    ComponentPose=@n3
    Alpha=1
    ConstraintWeights_0=0.3266
    ConstraintWeights_1:float=0.2386
    ConstraintWeights_2:float=0.1815
    ConstraintWeights_3:float=0.1155
    ConstraintWeights_4:float=0.0821
    ConstraintWeights_5:float=0.0557
   ) -> (Pose) {
    boneToModify {boneName=MSTR_pendant_circle_front}
    constraintSetup [
     {targetBone {boneName=deform_groin_l}}
     {targetBone {boneName=thigh_twist_01_l}}
     {targetBone {boneName=thigh_l}}
     {targetBone {boneName=deform_groin_base_l}}
     {targetBone {boneName=dyn_chain_down}}
     {targetBone {boneName=pelvis}}
    ]
    constraintWeights [1,1,1,1,1,1]
   }
   Output Pose(Result=Component To Local(ComponentPose=@n4)) {layerGroup=None}
   n5 = Control Rig!(
    Source=Copy Pose From Mesh() {bUseAttachedParent=true}
    Alpha
    IsCrouching2D:double=Get IsCrouching2D() -> (IsCrouching2D:double)
    IsCrouching:float=Get IsCrouchingAlpha() -> (IsCrouchingAlpha:double)
    IsMoving2DAlpha:double=Get Is Moving 2D Alpha() -> (Is Moving 2D Alpha:double)
   ) -> (Pose) {controlRigAssetReference {controlRigAsset=&/F_MED_VibrantShell_Chain_CtrlRig}}
   Save cached pose 'ControlRig'(Pose=@n5) {cacheName=ControlRig}
   n6 = "Transform (Modify) Bone - Bone: MSTR_bow"!(
    ComponentPose=Local To Component(LocalPose=Use cached pose 'ControlRig'())
    bAlphaBoolEnabled:bool=True
    Translation="-0.375,-0.290,0.106"
    Rotation="-13.497,-15.946,-27.116"
    Scale="1,1,1"
   ) -> (Pose) {
    alphaInputType=Bool
    rotationMode=BMM_Additive
    translationMode=BMM_Additive
    alphaBoolBlend {blendInTime=0.2, blendOutTime=0.4}
    boneToModify {boneName=MSTR_bow}
    rotation=<rot -13.497 -15.946 -27.116>
    translation=<-0.375 -0.29 0.106>
   }
   n7 = "Transform (Modify) Bone - Bone: MSTR_bow"(ComponentPose=@n6, bAlphaBoolEnabled=True, Translation="0.589,-1.917,1.494", Rotation="5.552,-10.952,-28.914", Scale="1,1,1") {
    alphaInputType=Bool
    rotationMode=BMM_Additive
    translationMode=BMM_Additive
    alphaBoolBlend {blendInTime=0.2, blendOutTime=0.4}
    boneToModify {boneName=MSTR_bow}
    rotation=<rot 5.552 -10.952 -28.914>
    translation=<0.589 -1.917 1.494>
   }
   n8 = "Transform (Modify) Bone - Bone: MSTR_chain_down"!(ComponentPose=@n7, bAlphaBoolEnabled:bool=True, Translation="0,-10,-7", Rotation="0,0,0", Scale="1,1,1") -> (Pose) {
    alphaInputType=Bool
    rotationMode=BMM_Additive
    rotationSpace=BCS_ParentBoneSpace
    translationMode=BMM_Additive
    translationSpace=BCS_ParentBoneSpace
    alphaBoolBlend {blendInTime=0.4, blendOutTime=0.2}
    boneToModify {boneName=MSTR_chain_down}
    translation {y=-10, z=-7}
   }
   n9 = RigidBody!(ComponentPose=@n8, Alpha, OverrideWorldGravity:Vector="0,0,0", ExternalForce) -> (Pose) {
    bClampLinearTranslationLimitToRefPose=true
    bOverrideWorldGravity=true
    overridePhysicsAsset=&/F_MED_VibrantShell_Chain_Physics
    simulationSpace=BaseBoneSpace
    baseBoneRef {boneName=pelvis}
    componentAppliedLinearAccClamp=<1000 1000 1000>
    overrideWorldGravity {z=-980}
    simSpaceSettings {dampingAlpha=0, maxAngularAcceleration=4.5, maxAngularVelocity=2.5, maxLinearAcceleration=1200, maxLinearVelocity=450, worldAlpha=1}
   }
   n10 = Anim Dynamics - Bone: dyn_skirt_01!(ComponentPose=@n9, Alpha=1, GravityOverride:Vector="0,0,-980", ExternalForce="0,0,0") -> (Pose) {
    angularDampingOverride=0.9999
    angularSpringConstant=3000
    bAngularSpring=true
    bOverrideAngularDamping=true
    bUseGravityOverride=true
    bUseSphericalLimits=true
    gravityScale=0.35
    simulationSpace=RootRelative
    boundBone {boneName=dyn_skirt_01}
    componentLinearAccScale=<1 1 1>
    componentLinearVelScale=<1 1 1>
    gravityOverride {z=-980}
    physicsBodyDefinitions [{
      collisionType=CustomSphere
      sphereCollisionRadius=11
      boundBone {boneName=dyn_skirt_01}
      boxExtents=<20 25 30>
      constraintSetup {angularLimitsMax=<15 15 15>, angularLimitsMin=<-15 0 -20>, angularTarget=<1 0 0>}
      localJointOffset=<2 1 0>
     }]
    simSpaceSettings {maxAngularAcceleration=32, maxAngularVelocity=8, simSpaceAngularAlpha=1}
    sphericalLimits [
     {limitRadius=6, drivingBone {boneName=calf_l}, sphereLocalOffset=<0 0 0>}
     {limitRadius=6, drivingBone {boneName=calf_l}, sphereLocalOffset=<-10 0 0>}
     {limitRadius=6, drivingBone {boneName=calf_l}, sphereLocalOffset=<-20 -1 0>}
    ]
   }
   n11 = Anim Dynamics - Bone: dyn_skirt_02!(ComponentPose=@n10, Alpha=1, GravityOverride:Vector="0,0,-980", ExternalForce="0,0,0") -> (Pose) {
    angularDampingOverride=0.999
    angularSpringConstant=3000
    bAngularSpring=true
    bOverrideAngularDamping=true
    bUseGravityOverride=true
    bUseSphericalLimits=true
    gravityScale=0.35
    simulationSpace=RootRelative
    boundBone {boneName=dyn_skirt_02}
    componentLinearAccScale=<1 1 1>
    componentLinearVelScale=<1 1 1>
    gravityOverride {z=-980}
    physicsBodyDefinitions [{
      collisionType=InnerSphere
      boundBone {boneName=dyn_skirt_02}
      boxExtents=<20 25 30>
      constraintSetup {angularLimitsMax=<4 30 15>, angularLimitsMin=<-4 0 -45>, angularTarget=<1 0 0>}
      localJointOffset=<3 0 0>
     }]
    relativeSpaceBone {boneName=dyn_skirt_01}
    simSpaceSettings {maxAngularAcceleration=40, maxAngularVelocity=8, simSpaceAngularAlpha=1}
    sphericalLimits [{limitRadius=7, drivingBone {boneName=foot_l}, sphereLocalOffset=<0 0 0>}]
   }
   n12 = Anim Dynamics - Bone: dyn_bracelet_1_l(ComponentPose=@n11, Alpha=1, ExternalForce="0,0,0") {
    angularDampingOverride=0.999
    angularSpringConstant=25
    bAngularSpring=true
    bGravityOverrideInSimSpace=true
    bOverrideAngularDamping=true
    bUseGravityOverride=true
    simulationSpace=RootRelative
    alphaBoolBlend {blendOption=Linear}
    boundBone {boneName=dyn_bracelet_1_l}
    componentAppliedLinearAccClamp=<3000 3000 3000>
    gravityOverride {z=-980}
    physicsBodyDefinitions [{
      sphereCollisionRadius=0
      boundBone {boneName=dyn_bracelet_1_l}
      boxExtents=<15 15 15>
      constraintSetup {angularLimitsMax=<10 2.5 10>, angularLimitsMin=<-10 -10 -10>, angularTarget=<0 0 0>}
      localJointOffset=<-0.1 0 0>
     }]
    simSpaceSettings {simSpaceAngularAlpha=1, externalAngularVelocity=<1 1 1>}
   }
   n13 = Anim Dynamics - Bone: dyn_bracelet_2_l(ComponentPose=@n12, Alpha=1, ExternalForce="0,0,0") {
    angularDampingOverride=0.999
    angularSpringConstant=25
    bAngularSpring=true
    bGravityOverrideInSimSpace=true
    bOverrideAngularDamping=true
    bUseGravityOverride=true
    simulationSpace=RootRelative
    alphaBoolBlend {blendOption=Linear}
    boundBone {boneName=dyn_bracelet_2_l}
    componentAppliedLinearAccClamp=<3000 3000 3000>
    gravityOverride {z=-980}
    physicsBodyDefinitions [{
      sphereCollisionRadius=0
      boundBone {boneName=dyn_bracelet_2_l}
      boxExtents=<15 15 15>
      constraintSetup {angularLimitsMax=<10 2.5 10>, angularLimitsMin=<-10 -10 -10>, angularTarget=<0 0 0>}
      localJointOffset=<-0.1 0 0>
     }]
    simSpaceSettings {simSpaceAngularAlpha=1, externalAngularVelocity=<1 1 1>}
   }
   n14 = Anim Dynamics - Bone: dyn_bracelet_r(ComponentPose=@n13, Alpha=1, ExternalForce="0,0,0") {
    angularDampingOverride=0.995
    angularSpringConstant=10
    bAngularSpring=true
    bGravityOverrideInSimSpace=true
    bOverrideAngularDamping=true
    bOverrideLinearDamping=true
    bUseGravityOverride=true
    linearDampingOverride=0.99
    simulationSpace=RootRelative
    alphaBoolBlend {blendOption=Linear}
    boundBone {boneName=dyn_bracelet_r}
    componentAppliedLinearAccClamp=<3000 3000 3000>
    gravityOverride {z=-980}
    physicsBodyDefinitions [{
      sphereCollisionRadius=0
      boundBone {boneName=dyn_bracelet_r}
      boxExtents=<10 10 10>
      constraintSetup {angularLimitsMax=<15 2.5 2.5>, angularLimitsMin=<-15 -2.5 -2.5>, angularTarget=<0 0 0>}
      localJointOffset=<1 0 0>
     }]
    simSpaceSettings {simSpaceAngularAlpha=1, externalAngularVelocity=<1 1 1>}
   }
   n15 = Anim Dynamics - Bone: dyn_ankle_bracelet_r(ComponentPose=@n14, Alpha=1, ExternalForce="0,0,0") {
    angularDampingOverride=0.995
    angularSpringConstant=10
    bAngularSpring=true
    bGravityOverrideInSimSpace=true
    bOverrideAngularDamping=true
    bUseGravityOverride=true
    simulationSpace=RootRelative
    alphaBoolBlend {blendOption=Linear}
    boundBone {boneName=dyn_ankle_bracelet_r}
    componentAppliedLinearAccClamp=<3000 3000 3000>
    gravityOverride {z=-980}
    physicsBodyDefinitions [{
      sphereCollisionRadius=0
      boundBone {boneName=dyn_ankle_bracelet_r}
      boxExtents=<30 30 30>
      constraintSetup {angularLimitsMax=<10 10 5>, angularLimitsMin=<-10 -10 -10>, angularTarget=<0 0 0>}
      localJointOffset=<0 -1 0>
     }]
    simSpaceSettings {simSpaceAngularAlpha=1, externalAngularVelocity=<1 1 1>}
   }
   n16 = Anim Dynamics - Bone: dyn_butterfly_wing_1_l(ComponentPose=@n15, Alpha=1, ExternalForce="0,0,0") {
    angularDampingOverride=0.995
    angularSpringConstant=30
    bAngularSpring=true
    bGravityOverrideInSimSpace=true
    bOverrideAngularDamping=true
    bOverrideLinearDamping=true
    bUseGravityOverride=true
    linearDampingOverride=0.9
    simulationSpace=RootRelative
    alphaBoolBlend {blendOption=Linear}
    boundBone {boneName=dyn_butterfly_wing_1_l}
    componentAppliedLinearAccClamp=<3000 3000 3000>
    gravityOverride {z=-980}
    physicsBodyDefinitions [{
      sphereCollisionRadius=0
      boundBone {boneName=dyn_butterfly_wing_1_l}
      boxExtents=<25 25 25>
      constraintSetup {angularLimitsMax=<10 10 5>, angularLimitsMin=<-10 -10 -10>, angularTarget=<0 0 0>}
      localJointOffset=<0.005 0 -0.25>
     }]
    simSpaceSettings {simSpaceAngularAlpha=1, externalAngularVelocity=<1 1 1>}
   }
   n17 = Anim Dynamics - Bone: dyn_butterfly_wing_2_l(ComponentPose=@n16, Alpha=1, ExternalForce="0,0,0") {
    angularDampingOverride=0.995
    angularSpringConstant=30
    bAngularSpring=true
    bGravityOverrideInSimSpace=true
    bOverrideAngularDamping=true
    bOverrideLinearDamping=true
    bUseGravityOverride=true
    linearDampingOverride=0.9
    simulationSpace=RootRelative
    alphaBoolBlend {blendOption=Linear}
    boundBone {boneName=dyn_butterfly_wing_2_l}
    componentAppliedLinearAccClamp=<3000 3000 3000>
    gravityOverride {z=-980}
    physicsBodyDefinitions [{
      sphereCollisionRadius=0
      boundBone {boneName=dyn_butterfly_wing_2_l}
      boxExtents=<25 25 25>
      constraintSetup {angularLimitsMax=<10 10 5>, angularLimitsMin=<-10 -10 -10>, angularTarget=<0 0 0>}
      localJointOffset=<0.005 0 0.25>
     }]
    simSpaceSettings {simSpaceAngularAlpha=1, externalAngularVelocity=<1 1 1>}
   }
   n18 = Anim Dynamics - Bone: dyn_butterfly_wing_1_r(ComponentPose=@n17, Alpha=1, ExternalForce="0,0,0") {
    angularDampingOverride=0.995
    angularSpringConstant=30
    bAngularSpring=true
    bGravityOverrideInSimSpace=true
    bOverrideAngularDamping=true
    bOverrideLinearDamping=true
    bUseGravityOverride=true
    linearDampingOverride=0.9
    simulationSpace=RootRelative
    alphaBoolBlend {blendOption=Linear}
    boundBone {boneName=dyn_butterfly_wing_1_r}
    componentAppliedLinearAccClamp=<3000 3000 3000>
    gravityOverride {z=-980}
    physicsBodyDefinitions [{
      sphereCollisionRadius=0
      boundBone {boneName=dyn_butterfly_wing_1_r}
      boxExtents=<25 25 25>
      constraintSetup {angularLimitsMax=<10 10 5>, angularLimitsMin=<-10 -10 -10>, angularTarget=<0 0 0>}
      localJointOffset=<0.005 0 -0.25>
     }]
    simSpaceSettings {simSpaceAngularAlpha=1, externalAngularVelocity=<1 1 1>}
   }
   n19 = Anim Dynamics - Bone: dyn_butterfly_wing_2_r(ComponentPose=@n18, Alpha=1, ExternalForce="0,0,0") {
    angularDampingOverride=0.995
    angularSpringConstant=30
    bAngularSpring=true
    bGravityOverrideInSimSpace=true
    bOverrideAngularDamping=true
    bOverrideLinearDamping=true
    bUseGravityOverride=true
    linearDampingOverride=0.9
    simulationSpace=RootRelative
    alphaBoolBlend {blendOption=Linear}
    boundBone {boneName=dyn_butterfly_wing_2_r}
    componentAppliedLinearAccClamp=<3000 3000 3000>
    gravityOverride {z=-980}
    physicsBodyDefinitions [{
      sphereCollisionRadius=0
      boundBone {boneName=dyn_butterfly_wing_2_r}
      boxExtents=<25 25 25>
      constraintSetup {angularLimitsMax=<10 10 5>, angularLimitsMin=<-10 -10 -10>, angularTarget=<0 0 0>}
      localJointOffset=<0.005 0 0.25>
     }]
    simSpaceSettings {simSpaceAngularAlpha=1, externalAngularVelocity=<1 1 1>}
   }
   n20 = Anim Dynamics - Bone: dyn_little_bow(ComponentPose=@n19, Alpha=1, ExternalForce="0,0,0") {
    angularDampingOverride=0.999
    angularSpringConstant=1200
    bAngularSpring=true
    bOverrideAngularDamping=true
    bUseGravityOverride=true
    gravityScale=0.35
    simulationSpace=RootRelative
    boundBone {boneName=dyn_little_bow}
    componentLinearAccScale=<1 1 1>
    componentLinearVelScale=<1 1 1>
    gravityOverride {z=-980}
    physicsBodyDefinitions [{
      boundBone {boneName=dyn_little_bow}
      boxExtents=<12 20 15>
      constraintSetup {angularLimitsMax=<0 40 20>, angularLimitsMin=<0 -40 -15>, angularTarget=<1 0 0>}
      localJointOffset=<1.5 0 0>
     }]
    simSpaceSettings {maxAngularAcceleration=40, maxAngularVelocity=8, simSpaceAngularAlpha=1, externalAngularVelocity=<1 0.6 0.3>}
   }
   n21 = Anim Dynamics - Bone: dyn_sleeve_down_l(ComponentPose=@n20, Alpha=1, ExternalForce="0,0,0") {
    angularDampingOverride=1
    angularSpringConstant=4200
    bAngularSpring=true
    bOverrideAngularDamping=true
    bUseGravityOverride=true
    gravityScale=0.35
    simulationSpace=RootRelative
    boundBone {boneName=dyn_sleeve_down_l}
    componentLinearAccScale=<1 1 1>
    componentLinearVelScale=<1 1 1>
    gravityOverride {z=-500}
    physicsBodyDefinitions [{
      boundBone {boneName=dyn_sleeve_down_l}
      boxExtents=<10 20 20>
      constraintSetup {angularLimitsMax=<2 10 5>, angularLimitsMin=<-2 -10 -20>, angularTarget=<1 0 0>}
      localJointOffset=<1.5 0 0>
     }]
    simSpaceSettings {maxAngularAcceleration=35, maxAngularVelocity=7, simSpaceAngularAlpha=1}
   }
   n22 = Anim Dynamics - Bone: dyn_sleeve_up_l(ComponentPose=@n21, Alpha=1, ExternalForce="0,0,0") {
    angularDampingOverride=1
    angularSpringConstant=4200
    bAngularSpring=true
    bOverrideAngularDamping=true
    bUseGravityOverride=true
    gravityScale=0.35
    simulationSpace=RootRelative
    boundBone {boneName=dyn_sleeve_up_l}
    componentLinearAccScale=<1 1 1>
    componentLinearVelScale=<1 1 1>
    gravityOverride {z=-500}
    physicsBodyDefinitions [{
      boundBone {boneName=dyn_sleeve_up_l}
      boxExtents=<10 20 20>
      constraintSetup {angularLimitsMax=<2 7 7>, angularLimitsMin=<-2 -15 -7>, angularTarget=<1 0 0>}
      localJointOffset=<0.5 0 0>
     }]
    simSpaceSettings {maxAngularAcceleration=35, maxAngularVelocity=7, simSpaceAngularAlpha=1}
   }
   n23 = Anim Dynamics - Bone: dyn_sleeve_back_l(ComponentPose=@n22, Alpha=1, ExternalForce="0,0,0") {
    angularDampingOverride=1
    angularSpringConstant=4200
    bAngularSpring=true
    bOverrideAngularDamping=true
    bUseGravityOverride=true
    gravityScale=0.35
    simulationSpace=RootRelative
    boundBone {boneName=dyn_sleeve_back_l}
    componentLinearAccScale=<1 1 1>
    componentLinearVelScale=<1 1 1>
    gravityOverride {z=-500}
    physicsBodyDefinitions [{
      boundBone {boneName=dyn_sleeve_back_l}
      boxExtents=<10 20 20>
      constraintSetup {angularLimitsMax=<2 4 5>, angularLimitsMin=<-2 -4 -5>, angularTarget=<1 0 0>}
      localJointOffset=<0.5 0 0>
     }]
    simSpaceSettings {maxAngularAcceleration=35, maxAngularVelocity=7, simSpaceAngularAlpha=1}
   }
   n24 = Anim Dynamics - Bone: dyn_sleeve_front_l(ComponentPose=@n23, Alpha=1, ExternalForce="0,0,0") {
    angularDampingOverride=1
    angularSpringConstant=4200
    bAngularSpring=true
    bOverrideAngularDamping=true
    bUseGravityOverride=true
    gravityScale=0.35
    simulationSpace=RootRelative
    boundBone {boneName=dyn_sleeve_front_l}
    componentLinearAccScale=<1 1 1>
    componentLinearVelScale=<1 1 1>
    gravityOverride {z=-500}
    physicsBodyDefinitions [{
      boundBone {boneName=dyn_sleeve_front_l}
      boxExtents=<10 17 17>
      constraintSetup {angularLimitsMax=<2 10 5>, angularLimitsMin=<-2 -10 -20>, angularTarget=<1 0 0>}
      localJointOffset=<1.5 0 0>
     }]
    simSpaceSettings {maxAngularAcceleration=35, maxAngularVelocity=7, simSpaceAngularAlpha=1}
   }
   n25 = Anim Dynamics - Bone: dyn_sleeve_down_r(ComponentPose=@n24, Alpha=1, ExternalForce="0,0,0") {
    angularDampingOverride=1
    angularSpringConstant=4200
    bAngularSpring=true
    bOverrideAngularDamping=true
    bUseGravityOverride=true
    gravityScale=0.35
    simulationSpace=RootRelative
    boundBone {boneName=dyn_sleeve_down_r}
    componentLinearAccScale=<1 1 1>
    componentLinearVelScale=<1 1 1>
    gravityOverride {z=-500}
    physicsBodyDefinitions [{
      boundBone {boneName=dyn_sleeve_down_r}
      boxExtents=<10 16 16>
      constraintSetup {angularLimitsMax=<2 10 5>, angularLimitsMin=<-2 -30 -20>, angularTarget=<1 0 0>}
      localJointOffset=<1.2 0 0>
     }]
    simSpaceSettings {maxAngularAcceleration=40, maxAngularVelocity=8, simSpaceAngularAlpha=1}
   }
   n26 = Anim Dynamics - Bone: dyn_sleeve_up_r(ComponentPose=@n25, Alpha=1, ExternalForce="0,0,0") {
    angularDampingOverride=1
    angularSpringConstant=4200
    bAngularSpring=true
    bOverrideAngularDamping=true
    bUseGravityOverride=true
    gravityScale=0.35
    simulationSpace=RootRelative
    boundBone {boneName=dyn_sleeve_up_r}
    componentLinearAccScale=<1 1 1>
    componentLinearVelScale=<1 1 1>
    gravityOverride {z=-500}
    physicsBodyDefinitions [{
      boundBone {boneName=dyn_sleeve_up_r}
      boxExtents=<10 20 20>
      constraintSetup {angularLimitsMax=<2 7 7>, angularLimitsMin=<-2 -7 -7>, angularTarget=<1 0 0>}
      localJointOffset=<0.35 0 0>
     }]
    simSpaceSettings {maxAngularAcceleration=30, maxAngularVelocity=6, simSpaceAngularAlpha=1}
   }
   n27 = Anim Dynamics - Bone: dyn_sleeve_back_r(ComponentPose=@n26, Alpha=1, ExternalForce="0,0,0") {
    angularDampingOverride=1
    angularSpringConstant=4200
    bAngularSpring=true
    bOverrideAngularDamping=true
    bUseGravityOverride=true
    gravityScale=0.35
    simulationSpace=RootRelative
    boundBone {boneName=dyn_sleeve_back_r}
    componentLinearAccScale=<1 1 1>
    componentLinearVelScale=<1 1 1>
    gravityOverride {z=-500}
    physicsBodyDefinitions [{
      boundBone {boneName=dyn_sleeve_back_r}
      boxExtents=<10 17 20>
      constraintSetup {angularLimitsMax=<2 4 5>, angularLimitsMin=<-2 -4 -5>, angularTarget=<1 0 0>}
      localJointOffset=<0.6 0 0>
     }]
    simSpaceSettings {maxAngularAcceleration=35, maxAngularVelocity=7, simSpaceAngularAlpha=1}
   }
   n28 = Anim Dynamics - Bone: dyn_sleeve_front_r(ComponentPose=@n27, Alpha=1, ExternalForce="0,0,0") {
    angularDampingOverride=1
    angularSpringConstant=4200
    bAngularSpring=true
    bOverrideAngularDamping=true
    bUseGravityOverride=true
    gravityScale=0.35
    simulationSpace=RootRelative
    boundBone {boneName=dyn_sleeve_front_r}
    componentLinearAccScale=<1 1 1>
    componentLinearVelScale=<1 1 1>
    gravityOverride {z=-500}
    physicsBodyDefinitions [{
      boundBone {boneName=dyn_sleeve_front_r}
      boxExtents=<10 17 14>
      constraintSetup {angularLimitsMax=<2 10 5>, angularLimitsMin=<-2 -10 -20>, angularTarget=<1 0 0>}
      localJointOffset=<1.3 0 0>
     }]
    simSpaceSettings {maxAngularAcceleration=35, maxAngularVelocity=7, simSpaceAngularAlpha=1}
   }
   n29 = "Constraint - MSTR_pendant_star_back (PPPPPP)"!(
    ComponentPose=@n28
    Alpha=1
    ConstraintWeights_0=0.5052
    ConstraintWeights_1:float=0.3379
    ConstraintWeights_2:float=0.0657
    ConstraintWeights_3:float=0.0588
    ConstraintWeights_4:float=0.0278
    ConstraintWeights_5:float=0.0045
   ) -> (Pose) {
    boneToModify {boneName=MSTR_pendant_star_back}
    constraintSetup [
     {targetBone {boneName=pelvis}}
     {targetBone {boneName=deform_glute_l}}
     {targetBone {boneName=thigh_twist_01_l}}
     {targetBone {boneName=deform_glute_r}}
     {targetBone {boneName=thigh_twist_01_r}}
     {targetBone {boneName=thigh_l}}
    ]
    constraintWeights [1,1,1,1,1,1]
   }
   n30 = "Constraint - MSTR_pendant_bunny_back (PPPPP)"!(
    ComponentPose=@n29
    Alpha=1
    ConstraintWeights_0=0.6405
    ConstraintWeights_1:float=0.2843
    ConstraintWeights_2:float=0.0546
    ConstraintWeights_3:float=0.0194
    ConstraintWeights_4:float=0.0012
   ) -> (Pose) {
    boneToModify {boneName=MSTR_pendant_bunny_back}
    constraintSetup [
     {targetBone {boneName=pelvis}}
     {targetBone {boneName=deform_glute_r}}
     {targetBone {boneName=thigh_twist_01_r}}
     {targetBone {boneName=deform_glute_l}}
     {targetBone {boneName=thigh_twist_01_l}}
    ]
    constraintWeights [1,1,1,1,1]
   }
   n31 = Anim Dynamics - Bone: dyn_pendant_star_back_01(ComponentPose=@n30, Alpha=1, ExternalForce="0,0,0") {
    angularDampingOverride=0.999
    angularSpringConstant=100
    bAngularSpring=true
    bOverrideAngularDamping=true
    bUseGravityOverride=true
    gravityScale=0.35
    simulationSpace=RootRelative
    boundBone {boneName=dyn_pendant_star_back_01}
    componentLinearAccScale=<1 1 1>
    componentLinearVelScale=<1 1 1>
    gravityOverride {z=-980}
    physicsBodyDefinitions [{
      boundBone {boneName=dyn_pendant_star_back_01}
      boxExtents=<10 14 14>
      constraintSetup {angularLimitsMax=<2 15 5>, angularLimitsMin=<-2 -5 -20>, angularTarget=<1 0 0>}
      localJointOffset=<4 0 0>
     }]
    simSpaceSettings {maxAngularAcceleration=32, maxAngularVelocity=8, simSpaceAngularAlpha=1, externalAngularVelocity=<1 0.6 0.3>}
   }
   n32 = Anim Dynamics - Bone: dyn_pendant_star_back_02(ComponentPose=@n31, Alpha=1, ExternalForce="0,0,0") {
    angularDampingOverride=0.999
    angularSpringConstant=100
    bAngularSpring=true
    bOverrideAngularDamping=true
    bUseGravityOverride=true
    gravityScale=0.35
    simulationSpace=RootRelative
    boundBone {boneName=dyn_pendant_star_back_02}
    componentLinearAccScale=<1 1 1>
    componentLinearVelScale=<1 1 1>
    gravityOverride {z=-980}
    physicsBodyDefinitions [{
      boundBone {boneName=dyn_pendant_star_back_02}
      boxExtents=<10 14 14>
      constraintSetup {angularLimitsMax=<2 20 5>, angularLimitsMin=<-2 0 -20>, angularTarget=<1 0 0>}
      localJointOffset=<4 0.2 0>
     }]
    simSpaceSettings {maxAngularAcceleration=32, maxAngularVelocity=8, simSpaceAngularAlpha=1, externalAngularVelocity=<1 0.6 0.3>}
   }
   n33 = Anim Dynamics - Bone: dyn_pendant_fox_back_01(ComponentPose=@n32, Alpha=1, ExternalForce="0,0,0") {
    angularDampingOverride=0.999
    angularSpringConstant=120
    bAngularSpring=true
    bOverrideAngularDamping=true
    bUseGravityOverride=true
    gravityScale=0.35
    simulationSpace=RootRelative
    boundBone {boneName=dyn_pendant_fox_back_01}
    componentLinearAccScale=<1 1 1>
    componentLinearVelScale=<1 1 1>
    gravityOverride {z=-980}
    physicsBodyDefinitions [{
      boundBone {boneName=dyn_pendant_fox_back_01}
      boxExtents=<25 19 28>
      constraintSetup {angularLimitsMax=<5 30 0>, angularLimitsMin=<-5 -15 -50>, angularTarget=<1 0 0>}
      localJointOffset=<3 0 0>
     }]
    simSpaceSettings {maxAngularAcceleration=35, maxAngularVelocity=7, simSpaceAngularAlpha=1, externalAngularVelocity=<0.8 0.8 0.8>}
   }
   n34 = Anim Dynamics - Bone: dyn_pendant_fox_back_02(ComponentPose=@n33, Alpha=1, ExternalForce="0,0,0") {
    angularDampingOverride=0.999
    angularSpringConstant=120
    bAngularSpring=true
    bOverrideAngularDamping=true
    bUseGravityOverride=true
    gravityScale=0.35
    simulationSpace=RootRelative
    boundBone {boneName=dyn_pendant_fox_back_02}
    componentLinearAccScale=<1 1 1>
    componentLinearVelScale=<1 1 1>
    gravityOverride {z=-980}
    physicsBodyDefinitions [{
      boundBone {boneName=dyn_pendant_fox_back_02}
      boxExtents=<20 24 28>
      constraintSetup {angularLimitsMax=<5 30 0>, angularLimitsMin=<-5 -15 -50>, angularTarget=<1 0 0>}
      localJointOffset=<3 0 0>
     }]
    simSpaceSettings {maxAngularAcceleration=35, maxAngularVelocity=7, simSpaceAngularAlpha=1}
   }
   n35 = Anim Dynamics - Bone: dyn_pendant_bunny_back_01(ComponentPose=@n34, Alpha=1, ExternalForce="0,0,0") {
    angularDampingOverride=0.999
    angularSpringConstant=100
    bAngularSpring=true
    bOverrideAngularDamping=true
    bUseGravityOverride=true
    gravityScale=0.35
    simulationSpace=RootRelative
    boundBone {boneName=dyn_pendant_bunny_back_01}
    componentLinearAccScale=<1 1 1>
    componentLinearVelScale=<1 1 1>
    gravityOverride {z=-980}
    physicsBodyDefinitions [{
      boundBone {boneName=dyn_pendant_bunny_back_01}
      boxExtents=<10 14 14>
      constraintSetup {angularLimitsMax=<2 15 15>, angularLimitsMin=<-2 -5 -10>, angularTarget=<1 0 0>}
      localJointOffset=<2.5 0 0>
     }]
    simSpaceSettings {maxAngularAcceleration=32, maxAngularVelocity=8, simSpaceAngularAlpha=1, externalAngularVelocity=<1 0.6 0.3>}
   }
   n36 = Anim Dynamics - Bone: dyn_pendant_bunny_back_02(ComponentPose=@n35, Alpha=1, ExternalForce="0,0,0") {
    angularDampingOverride=0.999
    angularSpringConstant=100
    bAngularSpring=true
    bOverrideAngularDamping=true
    bUseGravityOverride=true
    gravityScale=0.35
    simulationSpace=RootRelative
    boundBone {boneName=dyn_pendant_bunny_back_02}
    componentLinearAccScale=<1 1 1>
    componentLinearVelScale=<1 1 1>
    gravityOverride {z=-980}
    physicsBodyDefinitions [{
      boundBone {boneName=dyn_pendant_bunny_back_02}
      boxExtents=<10 14 14>
      constraintSetup {angularLimitsMax=<2 20 20>, angularLimitsMin=<-2 0 -4>, angularTarget=<1 0 0>}
      localJointOffset=<2.5 0.2 0>
     }]
    simSpaceSettings {maxAngularAcceleration=32, maxAngularVelocity=8, simSpaceAngularAlpha=1, externalAngularVelocity=<1 0.6 0.3>}
   }
   n37 = Anim Dynamics - Bone: dyn_pendant_circle_01_front(ComponentPose=@n36, Alpha=1, ExternalForce="0,0,0") {
    angularDampingOverride=1
    angularSpringConstant=120
    bAngularSpring=true
    bOverrideAngularDamping=true
    bUseGravityOverride=true
    bUseSphericalLimits=true
    gravityScale=0.35
    simulationSpace=RootRelative
    boundBone {boneName=dyn_pendant_circle_01_front}
    componentLinearAccScale=<1 1 1>
    componentLinearVelScale=<1 1 1>
    gravityOverride {z=-980}
    physicsBodyDefinitions [{
      collisionType=CustomSphere
      sphereCollisionRadius=3.6
      boundBone {boneName=dyn_pendant_circle_01_front}
      boxExtents=<15 15 15>
      constraintSetup {angularLimitsMax=<4 15 0>, angularLimitsMin=<-4 -15 -60>, angularTarget=<1 0 0>}
      localJointOffset=<2 0 0>
     }]
    simSpaceSettings {maxAngularAcceleration=35, maxAngularVelocity=7, simSpaceAngularAlpha=1}
    sphericalLimits [{limitRadius=4, drivingBone {boneName=thigh_l}, sphereLocalOffset=<-15 0 0>}]
   }
   n38 = Anim Dynamics - Bone: dyn_pendant_circle_02_front(ComponentPose=@n37, Alpha=1, ExternalForce="0,0,0") {
    angularDampingOverride=1
    angularSpringConstant=100
    bAngularSpring=true
    bOverrideAngularDamping=true
    bUseGravityOverride=true
    bUseSphericalLimits=true
    gravityScale=0.35
    simulationSpace=RootRelative
    boundBone {boneName=dyn_pendant_circle_02_front}
    componentLinearAccScale=<1 1 1>
    componentLinearVelScale=<1 1 1>
    gravityOverride {z=-980}
    physicsBodyDefinitions [{
      collisionType=CustomSphere
      sphereCollisionRadius=3.5
      boundBone {boneName=dyn_pendant_circle_02_front}
      boxExtents=<15 15 15>
      constraintSetup {angularLimitsMax=<3 10 10>, angularLimitsMin=<-3 -10 -10>, angularTarget=<1 0 0>}
      localJointOffset=<2 0 0>
     }]
    simSpaceSettings {maxAngularAcceleration=35, maxAngularVelocity=7, simSpaceAngularAlpha=1}
    sphericalLimits [
     {limitRadius=5.5, drivingBone {boneName=thigh_l}, sphereLocalOffset=<-17 0 0>}
     {limitRadius=6, drivingBone {boneName=thigh_l}, sphereLocalOffset=<-20.0008 0.0003 -1.1869>}
     {limitRadius=4, drivingBone {boneName=thigh_l}, sphereLocalOffset=<-17 -1 -4>}
    ]
   }
   n39 = Anim Dynamics - Bone: dyn_pendant_star_front_01(ComponentPose=@n38, Alpha=1, ExternalForce="0,0,0") {
    angularDampingOverride=1
    angularSpringConstant=230
    bAngularSpring=true
    bOverrideAngularDamping=true
    bUseGravityOverride=true
    gravityScale=0.35
    simulationSpace=RootRelative
    boundBone {boneName=dyn_pendant_star_front_01}
    componentLinearAccScale=<1 1 1>
    componentLinearVelScale=<1 1 1>
    gravityOverride {z=-980}
    physicsBodyDefinitions [{
      boundBone {boneName=dyn_pendant_star_front_01}
      boxExtents=<19 25 23>
      constraintSetup {angularLimitsMax=<0 15 15>, angularLimitsMin=<0 -15 -15>, angularTarget=<1 0 0>}
      localJointOffset=<4 0 0>
     }]
    simSpaceSettings {maxAngularAcceleration=35, maxAngularVelocity=7, simSpaceAngularAlpha=1}
   }
   n40 = Anim Dynamics - Bone: dyn_pendant_star_front_02(ComponentPose=@n39, Alpha=1, ExternalForce="0,0,0") {
    angularDampingOverride=0.9999
    angularSpringConstant=230
    bAngularSpring=true
    bOverrideAngularDamping=true
    bUseGravityOverride=true
    bUseSphericalLimits=true
    gravityScale=0.35
    simulationSpace=RootRelative
    boundBone {boneName=dyn_pendant_star_front_02}
    componentLinearAccScale=<1 1 1>
    componentLinearVelScale=<1 1 1>
    gravityOverride {z=-980}
    physicsBodyDefinitions [{
      sphereCollisionRadius=5
      boundBone {boneName=dyn_pendant_star_front_02}
      boxExtents=<19 20 20>
      constraintSetup {angularLimitsMax=<15 3 20>, angularLimitsMin=<-15 -15 -20>, angularTarget=<1 0 0>}
      localJointOffset=<4 0 0>
     }]
    simSpaceSettings {maxAngularAcceleration=35, maxAngularVelocity=7, simSpaceAngularAlpha=1}
    sphericalLimits [
     {limitRadius=6, drivingBone {boneName=pelvis}, sphereLocalOffset=<-15 0 0>}
     {limitRadius=3, drivingBone {boneName=thigh_twist_01_l}, sphereLocalOffset=<7 0 2>}
    ]
   }
   n41 = Anim Dynamics - Bone: dyn_pendant_bunny_front_01(ComponentPose=@n40, Alpha=1, ExternalForce="0,0,0") {
    angularDampingOverride=1
    angularSpringConstant=230
    bAngularSpring=true
    bOverrideAngularDamping=true
    bUseGravityOverride=true
    gravityScale=0.35
    simulationSpace=RootRelative
    boundBone {boneName=dyn_pendant_bunny_front_01}
    componentLinearAccScale=<1 1 1>
    componentLinearVelScale=<1 1 1>
    gravityOverride {z=-980}
    physicsBodyDefinitions [{
      boundBone {boneName=dyn_pendant_bunny_front_01}
      boxExtents=<19 25 23>
      constraintSetup {angularLimitsMax=<0 15 15>, angularLimitsMin=<0 -15 -15>, angularTarget=<1 0 0>}
      localJointOffset=<4 0 0>
     }]
    simSpaceSettings {maxAngularAcceleration=35, maxAngularVelocity=7, simSpaceAngularAlpha=1}
   }
   n42 = Anim Dynamics - Bone: dyn_pendant_bunny_front_02(ComponentPose=@n41, Alpha=1, ExternalForce="0,0,0") {
    angularDampingOverride=0.9999
    angularSpringConstant=230
    bAngularSpring=true
    bOverrideAngularDamping=true
    bUseGravityOverride=true
    bUseSphericalLimits=true
    gravityScale=0.35
    simulationSpace=RootRelative
    boundBone {boneName=dyn_pendant_bunny_front_02}
    componentLinearAccScale=<1 1 1>
    componentLinearVelScale=<1 1 1>
    gravityOverride {z=-980}
    physicsBodyDefinitions [{
      sphereCollisionRadius=4
      boundBone {boneName=dyn_pendant_bunny_front_02}
      boxExtents=<19 20 20>
      constraintSetup {angularLimitsMax=<15 3 20>, angularLimitsMin=<-15 -15 -20>, angularTarget=<1 0 0>}
      localJointOffset=<4 0 0>
     }]
    simSpaceSettings {maxAngularAcceleration=35, maxAngularVelocity=7, simSpaceAngularAlpha=1}
    sphericalLimits [{limitRadius=9.5, drivingBone {boneName=pelvis}, sphereLocalOffset=<-15 0 0>}]
   }
   n43 = Anim Dynamics - Bone: dyn_shirt_front(ComponentPose=@n42, Alpha=1, ExternalForce="0,0,0") {
    angularDampingOverride=1
    angularSpringConstant=145
    bAngularSpring=true
    bOverrideAngularDamping=true
    bUseGravityOverride=true
    gravityScale=0.35
    simulationSpace=RootRelative
    boundBone {boneName=dyn_shirt_front}
    componentLinearAccScale=<1 1 1>
    componentLinearVelScale=<1 1 1>
    gravityOverride {z=-980}
    physicsBodyDefinitions [{
      boundBone {boneName=dyn_shirt_front}
      boxExtents=<20 30 30>
      constraintSetup {angularLimitsMax=<15 15 15>, angularLimitsMin=<-15 -15 -15>, angularTarget=<1 0 0>}
      localJointOffset=<4 0 0>
     }]
    simSpaceSettings {maxAngularAcceleration=32, maxAngularVelocity=8, simSpaceAngularAlpha=1}
   }
   n44 = Anim Dynamics - Bone: dyn_shirt_l(ComponentPose=@n43, Alpha=1, ExternalForce="0,0,0") {
    angularDampingOverride=1
    angularSpringConstant=145
    bAngularSpring=true
    bOverrideAngularDamping=true
    bUseGravityOverride=true
    gravityScale=0.35
    simulationSpace=RootRelative
    boundBone {boneName=dyn_shirt_l}
    componentLinearAccScale=<1 1 1>
    componentLinearVelScale=<1 1 1>
    gravityOverride {z=-980}
    physicsBodyDefinitions [{
      boundBone {boneName=dyn_shirt_l}
      boxExtents=<20 30 30>
      constraintSetup {angularLimitsMax=<15 15 15>, angularLimitsMin=<-15 -15 -15>, angularTarget=<1 0 0>}
      localJointOffset=<4 0 0>
     }]
    simSpaceSettings {maxAngularAcceleration=32, maxAngularVelocity=8, simSpaceAngularAlpha=1}
   }
   n45 = Anim Dynamics - Bone: dyn_shirt_back(ComponentPose=@n44, Alpha=1, ExternalForce="0,0,0") {
    angularDampingOverride=1
    angularSpringConstant=145
    bAngularSpring=true
    bOverrideAngularDamping=true
    bUseGravityOverride=true
    gravityScale=0.35
    simulationSpace=RootRelative
    boundBone {boneName=dyn_shirt_back}
    componentLinearAccScale=<1 1 1>
    componentLinearVelScale=<1 1 1>
    gravityOverride {z=-980}
    physicsBodyDefinitions [{
      boundBone {boneName=dyn_shirt_back}
      boxExtents=<20 30 30>
      constraintSetup {angularLimitsMax=<15 15 15>, angularLimitsMin=<-15 -15 -15>, angularTarget=<1 0 0>}
      localJointOffset=<4 0 0>
     }]
    simSpaceSettings {maxAngularAcceleration=32, maxAngularVelocity=8, simSpaceAngularAlpha=1}
   }
   n46 = Anim Dynamics - Bone: dyn_shirt_r(ComponentPose=@n45, Alpha=1, ExternalForce="0,0,0") {
    angularDampingOverride=1
    angularSpringConstant=145
    bAngularSpring=true
    bOverrideAngularDamping=true
    bUseGravityOverride=true
    gravityScale=0.35
    simulationSpace=RootRelative
    boundBone {boneName=dyn_shirt_r}
    componentLinearAccScale=<1 1 1>
    componentLinearVelScale=<1 1 1>
    gravityOverride {z=-980}
    physicsBodyDefinitions [{
      boundBone {boneName=dyn_shirt_r}
      boxExtents=<20 30 30>
      constraintSetup {angularLimitsMax=<15 15 15>, angularLimitsMin=<-15 -15 -15>, angularTarget=<1 0 0>}
      localJointOffset=<4 0 0>
     }]
    simSpaceSettings {maxAngularAcceleration=32, maxAngularVelocity=8, simSpaceAngularAlpha=1}
   }
   n47 = Anim Dynamics - Bone: dyn_chain_up(ComponentPose=@n46, Alpha=1, ExternalForce="0,0,0") {
    angularDampingOverride=1
    angularSpringConstant=4000
    bAngularSpring=true
    bOverrideAngularDamping=true
    bUseGravityOverride=true
    gravityScale=0.35
    simulationSpace=RootRelative
    boundBone {boneName=dyn_chain_up}
    componentLinearAccScale=<1 1 1>
    componentLinearVelScale=<1 1 1>
    gravityOverride {z=-980}
    physicsBodyDefinitions [{
      boundBone {boneName=dyn_chain_up}
      boxExtents=<15 15 15>
      constraintSetup {angularLimitsMax=<10 10 10>, angularLimitsMin=<-10 -45 -10>, angularTarget=<1 0 0>}
      localJointOffset=<4 0 0>
     }]
    simSpaceSettings {maxAngularAcceleration=40, maxAngularVelocity=8, simSpaceAngularAlpha=1}
   }
   n48 = Anim Dynamics - Bone: dyn_chain_down(ComponentPose=@n47, Alpha=1, ExternalForce="0,0,0") {
    angularDampingOverride=1
    angularSpringConstant=4000
    bAngularSpring=true
    bOverrideAngularDamping=true
    bUseGravityOverride=true
    gravityScale=0.35
    simulationSpace=RootRelative
    boundBone {boneName=dyn_chain_down}
    componentLinearAccScale=<1 1 1>
    componentLinearVelScale=<1 1 1>
    gravityOverride {z=-980}
    physicsBodyDefinitions [{
      boundBone {boneName=dyn_chain_down}
      boxExtents=<18 18 18>
      constraintSetup {angularLimitsMax=<5 10 5>, angularLimitsMin=<-5 -45 -5>, angularTarget=<1 0 0>}
      localJointOffset=<4 0 0>
     }]
    simSpaceSettings {maxAngularAcceleration=35, maxAngularVelocity=7, simSpaceAngularAlpha=1}
   }
   Save cached pose 'Dynamics'(Pose=Component To Local(ComponentPose=@n48)) {cacheName=Dynamics}
   n49 = "Transform (Modify) Bone - Bone: correction_chain_up"!(
    ComponentPose=Local To Component(LocalPose=Use cached pose 'Dynamics'())
    bAlphaBoolEnabled:bool=True
    Translation="0.000,3.077,10.063"
    Rotation="0,0,0"
    Scale="1,1,1"
   ) -> (Pose) {
    alphaInputType=Bool
    rotationMode=BMM_Additive
    translationMode=BMM_Additive
    alphaBoolBlend {blendInTime=0.2, blendOutTime=0.4}
    boneToModify {boneName=correction_chain_up}
    translation {y=3.077, z=10.063}
   }
   n50 = "Transform (Modify) Bone - Bone: MSTR_chain_down"(ComponentPose=@n49, bAlphaBoolEnabled=True, Translation="0.540,0.000,8.087", Rotation="0,0,0", Scale="1,1,1") {
    alphaInputType=Bool
    rotationMode=BMM_Additive
    translationMode=BMM_Additive
    alphaBoolBlend {blendInTime=0.2, blendOutTime=0.4}
    boneToModify {boneName=MSTR_chain_down}
    translation {x=0.54, z=8.087}
   }
   n51 = "Transform (Modify) Bone - Bone: dyn_pendant_fox_back_01"(ComponentPose=@n50, Alpha=1, Translation="0,0,0", Rotation="20,-35,0", Scale="1,1,1") {
    rotationMode=BMM_Additive
    rotationSpace=BCS_ParentBoneSpace
    translationMode=BMM_Additive
    translationSpace=BCS_BoneSpace
    alphaBoolBlend {blendInTime=0.2, blendOutTime=0.4}
    alphaScaleBiasClamp {bInterpResult=true, interpSpeedIncreasing=7}
    boneToModify {boneName=dyn_pendant_fox_back_01}
    rotation {pitch=20, yaw=-35}
   }
   n52 = "Transform (Modify) Bone - Bone: dyn_pendant_circle_01_front"!(ComponentPose=@n51, bAlphaBoolEnabled:bool=True, Translation="0,0,0", Rotation="-2.828,-9.992,0.962", Scale="1,1,1") -> (Pose) {
    alphaInputType=Bool
    rotationMode=BMM_Additive
    rotationSpace=BCS_ParentBoneSpace
    translationMode=BMM_Additive
    translationSpace=BCS_BoneSpace
    alphaBoolBlend {blendInTime=0.2, blendOutTime=0.4}
    boneToModify {boneName=dyn_pendant_circle_01_front}
    rotation=<rot -2.828 -9.992 0.962>
   }
   n53 = "Transform (Modify) Bone - Bone: dyn_pendant_circle_01_front"(ComponentPose=@n52, bAlphaBoolEnabled=True, Translation="0,0,0", Rotation="0.008,-9.455,1.825", Scale="1,1,1") {
    alphaInputType=Bool
    rotationMode=BMM_Additive
    rotationSpace=BCS_ParentBoneSpace
    translationMode=BMM_Additive
    translationSpace=BCS_BoneSpace
    alphaBoolBlend {blendInTime=0.2, blendOutTime=0.4}
    boneToModify {boneName=dyn_pendant_circle_01_front}
    rotation=<rot 0.008 -9.455 1.825>
   }
   n54 = "Transform (Modify) Bone - Bone: dyn_pendant_circle_01_front"(ComponentPose=@n53, bAlphaBoolEnabled=True, Translation="0,0,0", Rotation="22.386,-21.299,15.862", Scale="1,1,1") {
    alphaInputType=Bool
    rotationMode=BMM_Additive
    rotationSpace=BCS_ParentBoneSpace
    translationMode=BMM_Additive
    translationSpace=BCS_BoneSpace
    alphaBoolBlend {blendInTime=0.2, blendOutTime=0.4}
    boneToModify {boneName=dyn_pendant_circle_01_front}
    rotation=<rot 22.386 -21.299 15.862>
   }
   n55 = "Transform (Modify) Bone - Bone: dyn_pendant_circle_02_front"!(ComponentPose=@n54, bAlphaBoolEnabled:bool=True, Translation="-0.272,0.294,0.686", Rotation="-14.827,-8.943,9.872", Scale="1,1,1") -> (Pose) {
    alphaInputType=Bool
    rotationMode=BMM_Additive
    rotationSpace=BCS_ParentBoneSpace
    translationMode=BMM_Additive
    translationSpace=BCS_BoneSpace
    alphaBoolBlend {blendInTime=0.2, blendOutTime=0.4}
    boneToModify {boneName=dyn_pendant_circle_02_front}
    rotation=<rot -14.827 -8.943 9.872>
    translation=<-0.272 0.294 0.686>
   }
   n56 = "Transform (Modify) Bone - Bone: dyn_pendant_star_front_01"(ComponentPose=@n55, Alpha=1, Translation="0,0,0", Rotation="24.757,-59.883,-65.898", Scale="1,1,1") {
    rotationMode=BMM_Additive
    rotationSpace=BCS_BoneSpace
    translationMode=BMM_Additive
    translationSpace=BCS_BoneSpace
    alphaScaleBiasClamp {bInterpResult=true}
    boneToModify {boneName=dyn_pendant_star_front_01}
    rotation=<rot 24.757 -59.883 -65.898>
   }
   n57 = "Transform (Modify) Bone - Bone: dyn_pendant_star_front_01"(ComponentPose=@n56, Alpha=1, Translation="0,0,0", Rotation="22.639,-15.109,12.123", Scale="1,1,1") {
    rotationMode=BMM_Additive
    rotationSpace=BCS_BoneSpace
    translationMode=BMM_Additive
    translationSpace=BCS_BoneSpace
    alphaScaleBiasClamp {bInterpResult=true}
    boneToModify {boneName=dyn_pendant_star_front_01}
    rotation=<rot 22.639 -15.109 12.123>
   }
   n58 = "Transform (Modify) Bone - Bone: correction_chain_down_m"(ComponentPose=@n57, Alpha=1, Translation="1.285,6.896,5.265", Rotation="3.724,-6.444,-30.140", Scale="1,1,1") {
    rotationMode=BMM_Additive
    rotationSpace=BCS_ParentBoneSpace
    translationMode=BMM_Additive
    translationSpace=BCS_ParentBoneSpace
    alphaBoolBlend {blendInTime=0.2, blendOutTime=0.4}
    alphaScaleBiasClamp {interpSpeedIncreasing=2}
    boneToModify {boneName=correction_chain_down_m}
    rotation=<rot 3.724 -6.444 -30.14>
    translation=<1.285 6.896 5.265>
   }
   n59 = "Transform (Modify) Bone - Bone: dyn_pendant_bunny_front_01"(ComponentPose=@n58, Alpha=1, Translation="0,0,0", Rotation="-10.588,-42.526,18.802", Scale="1,1,1") {
    rotationMode=BMM_Additive
    rotationSpace=BCS_BoneSpace
    translationMode=BMM_Additive
    translationSpace=BCS_BoneSpace
    alphaScaleBiasClamp {bInterpResult=true, interpSpeedIncreasing=15}
    boneToModify {boneName=dyn_pendant_bunny_front_01}
    rotation=<rot -10.588 -42.526 18.802>
   }
   n60 = "Transform (Modify) Bone - Bone: dyn_pendant_bunny_front_01"(ComponentPose=@n59, Alpha=1, Translation="0,0,0", Rotation="83.473,49.857,-28.567", Scale="1,1,1") {
    rotationMode=BMM_Additive
    rotationSpace=BCS_BoneSpace
    translationMode=BMM_Additive
    translationSpace=BCS_BoneSpace
    alphaScaleBiasClamp {bInterpResult=true}
    boneToModify {boneName=dyn_pendant_bunny_front_01}
    rotation=<rot 83.473 49.857 -28.567>
   }
   n61 = "Transform (Modify) Bone - Bone: dyn_pendant_bunny_front_01"(ComponentPose=@n60, Alpha=Get IsCrouchingAlpha(), Translation="0,0,0", Rotation="-73.081,-147.539,114.449", Scale="1,1,1") {
    rotationMode=BMM_Additive
    rotationSpace=BCS_BoneSpace
    translationMode=BMM_Additive
    translationSpace=BCS_BoneSpace
    alphaScaleBiasClamp {bInterpResult=true, interpSpeedIncreasing=2}
    boneToModify {boneName=dyn_pendant_bunny_front_01}
    rotation=<rot -73.081 -147.539 114.449>
   }
   n62 = "Transform (Modify) Bone - Bone: dyn_pendant_bunny_front_01"(ComponentPose=@n61, Alpha=1, Translation="0,0,0", Rotation="71.794,179.998,179.997", Scale="1,1,1") {
    rotationMode=BMM_Additive
    rotationSpace=BCS_BoneSpace
    translationMode=BMM_Additive
    translationSpace=BCS_BoneSpace
    alphaScaleBiasClamp {bInterpResult=true}
    boneToModify {boneName=dyn_pendant_bunny_front_01}
    rotation=<rot 71.794 179.998 179.997>
   }
   n63 = "Transform (Modify) Bone - Bone: MSTR_skirt_back"(ComponentPose=@n62, Alpha=1, Translation="3.94,-7.271,-3", Rotation="0.003,-16.859,-0.002", Scale="1,1,1") {
    rotationMode=BMM_Additive
    rotationSpace=BCS_BoneSpace
    translationMode=BMM_Additive
    translationSpace=BCS_BoneSpace
    alphaScaleBiasClamp {bInterpResult=true}
    boneToModify {boneName=MSTR_skirt_back}
    rotation=<rot 0.003 -16.859 -0.002>
    translation=<3.94 -7.271 -3>
   }
   n64 = "Transform (Modify) Bone - Bone: MSTR_skirt_back"(ComponentPose=@n63, Alpha=1, Translation="3.940,-7.271,-1.071", Rotation="0.003,-16.859,-0.002", Scale="1,1,1") {
    rotationMode=BMM_Additive
    rotationSpace=BCS_BoneSpace
    translationMode=BMM_Additive
    translationSpace=BCS_BoneSpace
    alphaScaleBiasClamp {bInterpResult=true}
    boneToModify {boneName=MSTR_skirt_back}
    rotation=<rot 0.003 -16.859 -0.002>
    translation=<3.94 -7.271 -1.071>
   }
   n65 = "Transform (Modify) Bone - Bone: MSTR_skirt_front"(ComponentPose=@n64, Alpha=1, Translation="1.152,3.540,-0.165", Rotation="-1.516,-8.222,-12.700", Scale="1,1,1") {
    rotationMode=BMM_Additive
    rotationSpace=BCS_BoneSpace
    translationMode=BMM_Additive
    translationSpace=BCS_BoneSpace
    alphaScaleBiasClamp {bInterpResult=true, interpSpeedDecreasing=1}
    boneToModify {boneName=MSTR_skirt_front}
    rotation=<rot -1.516 -8.222 -12.7>
    translation=<1.152 3.54 -0.165>
   }
   n66 = "Transform (Modify) Bone - Bone: MSTR_skirt_front"(ComponentPose=@n65, Alpha=1, Translation="0.361,-0.560,-2.113", Rotation="0.700,27.590,27.673", Scale="1,1,1") {
    rotationMode=BMM_Additive
    rotationSpace=BCS_BoneSpace
    translationMode=BMM_Additive
    translationSpace=BCS_BoneSpace
    alphaScaleBiasClamp {bInterpResult=true}
    boneToModify {boneName=MSTR_skirt_front}
    rotation=<rot 0.7 27.59 27.673>
    translation=<0.361 -0.56 -2.113>
   }
   n67 = "Transform (Modify) Bone - Bone: MSTR_skirt_front"(ComponentPose=@n66, Alpha=1, Translation="2,0,-0.6", Rotation="0,0,0", Scale="1,1,1") {
    rotationMode=BMM_Additive
    rotationSpace=BCS_BoneSpace
    translationMode=BMM_Additive
    translationSpace=BCS_BoneSpace
    alphaScaleBiasClamp {bInterpResult=true}
    boneToModify {boneName=MSTR_skirt_front}
    translation {x=2, z=-0.6}
   }
   n68 = "Transform (Modify) Bone - Bone: MSTR_skirt_front"(ComponentPose=@n67, Alpha=1, Translation="0.000,1.800,0.324", Rotation="-0.149,-2.319,-5.310", Scale="1,1,1") {
    rotationMode=BMM_Additive
    rotationSpace=BCS_BoneSpace
    translationMode=BMM_Additive
    translationSpace=BCS_BoneSpace
    alphaScaleBiasClamp {bInterpResult=true}
    boneToModify {boneName=MSTR_skirt_front}
    rotation=<rot -0.149 -2.319 -5.31>
    translation=<0 1.8 0.3241>
   }
   n69 = "Transform (Modify) Bone - Bone: MSTR_skirt_back"(ComponentPose=@n68, Alpha=1, Translation="0.000,-0.007,-4.433", Rotation="0,0,0", Scale="1,1,1") {
    rotationMode=BMM_Additive
    rotationSpace=BCS_BoneSpace
    translationMode=BMM_Additive
    translationSpace=BCS_BoneSpace
    alphaScaleBiasClamp {bInterpResult=true}
    boneToModify {boneName=MSTR_skirt_back}
    translation {y=-0.007, z=-4.433}
   }
   n70 = "Transform (Modify) Bone - Bone: MSTR_skirt_back"(ComponentPose=@n69, Alpha=1, Translation="0.000,-0.007,-4.433", Rotation="0,0,0", Scale="1,1,1") {
    rotationMode=BMM_Additive
    rotationSpace=BCS_BoneSpace
    translationMode=BMM_Additive
    translationSpace=BCS_BoneSpace
    alphaScaleBiasClamp {bInterpResult=true}
    boneToModify {boneName=MSTR_skirt_back}
    translation {y=-0.007, z=-4.433}
   }
   n71 = "Transform (Modify) Bone - Bone: dyn_pendant_fox_back_01"(ComponentPose=@n70, Alpha=1, Translation="0,-0.007,0", Rotation="0,-5,0", Scale="1,1,1") {
    rotationMode=BMM_Additive
    rotationSpace=BCS_BoneSpace
    translationMode=BMM_Additive
    translationSpace=BCS_BoneSpace
    alphaScaleBiasClamp {bInterpResult=true, interpSpeedDecreasing=14, interpSpeedIncreasing=25}
    boneToModify {boneName=dyn_pendant_fox_back_01}
    rotation {yaw=-5}
    translation {y=-0.007}
   }
   n72 = "Transform (Modify) Bone - Bone: dyn_pendant_fox_back_01"(ComponentPose=@n71, Alpha=1, Translation="0,-0.007,0", Rotation="0,-5,0", Scale="1,1,1") {
    rotationMode=BMM_Additive
    rotationSpace=BCS_BoneSpace
    translationMode=BMM_Additive
    translationSpace=BCS_BoneSpace
    alphaScaleBiasClamp {bInterpResult=true, interpSpeedDecreasing=14, interpSpeedIncreasing=25}
    boneToModify {boneName=dyn_pendant_fox_back_01}
    rotation {yaw=-5}
    translation {y=-0.007}
   }
   n73 = "Transform (Modify) Bone - Bone: dyn_pendant_star_back_01"(ComponentPose=@n72, Alpha=1, Translation="0,0,0", Rotation="0.144,-24.014,0.385", Scale="1,1,1") {
    rotationMode=BMM_Additive
    rotationSpace=BCS_BoneSpace
    translationMode=BMM_Additive
    translationSpace=BCS_BoneSpace
    alphaScaleBiasClamp {bInterpResult=true}
    boneToModify {boneName=dyn_pendant_star_back_01}
    rotation=<rot 0.144 -24.014 0.385>
   }
   n74 = "Transform (Modify) Bone - Bone: dyn_pendant_star_back_01"(ComponentPose=@n73, Alpha=1, Translation="0,0,0", Rotation="-1.484,-25.474,7.368", Scale="1,1,1") {
    rotationMode=BMM_Additive
    rotationSpace=BCS_BoneSpace
    translationMode=BMM_Additive
    translationSpace=BCS_BoneSpace
    alphaScaleBiasClamp {bInterpResult=true}
    boneToModify {boneName=dyn_pendant_star_back_01}
    rotation=<rot -1.484 -25.474 7.368>
   }
   n75 = "Transform (Modify) Bone - Bone: MSTR_skirt_back"(ComponentPose=@n74, Alpha=1, Translation="0,0,-4", Rotation="0,0,0", Scale="1,1,1") {
    rotationMode=BMM_Additive
    rotationSpace=BCS_BoneSpace
    translationMode=BMM_Additive
    translationSpace=BCS_BoneSpace
    alphaScaleBiasClamp {bInterpResult=true, interpSpeedIncreasing=15}
    boneToModify {boneName=MSTR_skirt_back}
    translation {z=-4}
   }
   Save cached pose 'Adjustments'(Pose=Component To Local(ComponentPose=@n75)) {cacheName=Adjustments}
  ]}
 EventGraph {nodes [
   Event BlueprintUpdateAnimation()
   > n1 = Sequence!(execute) -> (then_0, then_1, then_2:Exec, then_3:Exec)
   > n2 = "Gravity Override State Function (Parameter Asset) (Nativized)"(
    CharacterPartAnimInstance=Self-Reference()
    GameTimeInSeconds=GetGameTimeInSeconds()
    Parameters=/BRCosmetics/Characters/Player/Female/Medium/Bodies/F_MED_VibrantShell/Meshes/F_MED_VibrantShell_Parameters
   )
   > Set Override World Gravity(Override World Gravity:Vector=@n2) -> (Output_Get:Vector)
   n3 = Get bIsMoving2D()
   n4 = AND Boolean(A=Get bIsCrouching(), B=@n3)
   Set IsCrouching2D(execute=@n1.then_1, IsCrouching2D:double="To Float (Boolean)"(InBool=@n4)) -> (Output_Get:double)
   > Set Is Moving 2D Alpha(Is Moving 2D Alpha:double="To Float (Boolean)"(InBool=@n3)) -> (Output_Get:double)
   n5 = NOT Boolean(A=@n4)
   n6 = Get bIsCrouching()
   > Set IsCrouchingAlpha(IsCrouchingAlpha:double="To Float (Boolean)"(InBool=AND Boolean(A=@n6, B=@n5))) -> (Output_Get:double)
   n7 = NOT Boolean(A=Get bIsCrouching())
   n8 = Get bIsMoving2D()
   > Set 2DMovingStand(2DMovingStand:double="To Float (Boolean)"(InBool=AND Boolean(A=@n7, B=@n8))) -> (Output_Get:double)
   > Set LLegRaisedSide(LLegRaisedSide:double=GetCurveValue(CurveName=LLegRaisedSide)) -> (Output_Get:double)
   > Set RLegRaised(RLegRaised:double=GetCurveValue(CurveName=RLegRaised)) -> (Output_Get:double)
   > Set LLegRaised(LLegRaised:double=GetCurveValue(CurveName=LLegRaised)) -> (Output_Get:double)
   n9 = NOT Boolean(A=Get bIsSkydiveFloating())
   n10 = AND Boolean(A=NOT Boolean(A=Get bIsMoving2D()), B=OR Boolean(A=Get bIsJumping(), B=Get bIsFalling()))
   > Set JumpNotMoving(JumpNotMoving:double="To Float (Boolean)"(InBool=AND Boolean(A=@n10, B=@n9))) -> (Output_Get:double)
   Set torsoLeftRotationOnPickaxeAlpha(
    execute=@n1.then_2
    torsoLeftRotationOnPickaxeAlpha:float=GetAlphaFromReferenceTransformDelta(
     SkeletalMeshComponent=GetOwnerSkeletalMeshComponent()
     BoneName=spine_05
     BaseName=pelvis
     DeltaComponentType=EulerY
     InputRange=(X=0,Y=-15)
     TwistAxis="1,0,0"
    )
   ) -> (Output_Get:float)
  ]}
]
variables [
 2DMovingStand {type=number}
 Is Moving 2D Alpha {type=number}
 IsCrouching2D {type=number}
 IsCrouchingAlpha {type=number}
 JumpNotMoving {type=number}
 LLegRaised {type=number}
 LLegRaisedSide {type=number}
 Override World Gravity {type=Vector}
 RLegRaised {type=number}
 torsoLeftRotationOnPickaxeAlpha {type=number}
]
