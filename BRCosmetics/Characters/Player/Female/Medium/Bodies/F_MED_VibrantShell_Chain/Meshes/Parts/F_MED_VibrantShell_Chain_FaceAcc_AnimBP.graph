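# Post-pose AnimBP for the Vibrant Shell Chain face accessory that animates its ponytail and hat correction bones.
# It copies the parent mesh pose, runs a Control Rig driven by crouch and ADS, then simulates RigidBody physics relative to neck_02.
# Constraints and additive bone offsets follow, mostly weighted by backbling and HeadUp curve to reposition the ponytail.
# EventGraph updates those alphas, a parameterized gravity override, and apparently unused arm and torso alphas.
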AnimBlueprint /BRCosmetics/Characters/Player/Female/Medium/Bodies/F_MED_VibrantShell_Chain/Meshes/Parts/F_MED_VibrantShell_Chain_FaceAcc_AnimBP
target_skeleton=&/F_MED_VibrantShell_Chain_FaceAcc_Skeleton
graphs [
 AnimGraph {nodes [
   n1 = Control Rig!(
    Source=Copy Pose From Mesh() {bUseAttachedParent=true}
    Alpha
    Crouching:double=Get Is Crouching Alpha() -> (Is Crouching Alpha:double)
    ADS:float=Get ADS() -> (ADS:double)
   ) -> (Pose) {controlRigAssetReference {controlRigAsset=&/F_MED_VibrantShell_Chain_FaceAcc_CtrlRig}}
   n2 = RigidBody!(ComponentPose=Local To Component(LocalPose=@n1), bAlphaBoolEnabled:bool, OverrideWorldGravity:Vector="0,0,0", ExternalForce) -> (Pose) {
    alphaInputType=Bool
    bClampLinearTranslationLimitToRefPose=true
    bOverrideWorldGravity=true
    overridePhysicsAsset=&/F_MED_VibrantShell_Chain_FaceAcc_Physics
    simulationSpace=BaseBoneSpace
    baseBoneRef {boneName=neck_02}
    componentAppliedLinearAccClamp=<750 750 750>
    overrideWorldGravity {z=-980}
    simSpaceSettings {dampingAlpha=0, maxAngularAcceleration=15, maxAngularVelocity=3.5, maxLinearAcceleration=1200, maxLinearVelocity=100, worldAlpha=1}
   }
   n3 = Get IsWearingBackbling() -> (IsWearingBackbling:double)
   n4 = "Constraint - dyn_ponytail_01 (PP)"!(ComponentPose=@n2, Alpha=@n3, ConstraintWeights_0=0.6, ConstraintWeights_1:float=0.4) -> (Pose) {
    boneToModify {boneName=dyn_ponytail_01}
    constraintSetup [{targetBone {boneName=head}}, {targetBone {boneName=spine_05}}]
    constraintWeights [1,1]
   }
   n5 = "Constraint - dyn_ponytail_02 (PPPPP)"!(
    ComponentPose=@n4
    Alpha=@n3
    ConstraintWeights_0=0.2
    ConstraintWeights_1:float=0
    ConstraintWeights_2:float=0
    ConstraintWeights_3:float=0
    ConstraintWeights_4:float=0.8
   ) -> (Pose) {
    boneToModify {boneName=dyn_ponytail_02}
    constraintSetup [
     {targetBone {boneName=head}}
     {targetBone {boneName=neck_01}}
     {targetBone {boneName=clavicle_l}}
     {targetBone {boneName=clavicle_r}}
     {targetBone {boneName=spine_05}}
    ]
    constraintWeights [1,1,1,1,1]
   }
   n6 = "Constraint - dyn_ponytail_03 (PP)"!(ComponentPose=@n5, Alpha=@n3, ConstraintWeights_0=1, ConstraintWeights_1:float=0) -> (Pose) {
    boneToModify {boneName=dyn_ponytail_03}
    constraintSetup [{targetBone {boneName=spine_05}}, {targetBone {boneName=head}}]
    constraintWeights [1,1]
   }
   n7 = "Transform (Modify) Bone - Bone: MSTR_ponytail"(
    ComponentPose=@n6
    Alpha=Get HeadUpAlpha() -> (HeadUpAlpha:double)
    Translation="0.000,-2.721,3.254"
    Rotation="-20.095,0.000,-0.000"
    Scale="1,1,1"
   ) {
    rotationMode=BMM_Additive
    translationMode=BMM_Additive
    boneToModify {boneName=MSTR_ponytail}
    rotation {pitch=-20.095}
    translation {y=-2.721, z=3.254}
   }
   n8 = "Transform (Modify) Bone - Bone: dyn_ponytail_03"(ComponentPose=@n7, Alpha=Get IsWearingBackbling(), Translation="0,4,0", Rotation="0,0,15", Scale="1,1,1") {rotationMode=BMM_Additive, translationMode=BMM_Additive, boneToModify {boneName=dyn_ponytail_03}, rotation {roll=15}, translation {y=4}}
   n9 = Get NewVar() -> (NewVar:double)
   n10 = "Transform (Modify) Bone - Bone: MSTR_ponytail"(ComponentPose=@n8, Alpha=@n9, Translation="0,0,-4", Rotation="0,0,10", Scale="1,1,1") {rotationMode=BMM_Additive, translationMode=BMM_Additive, boneToModify {boneName=MSTR_ponytail}, rotation {roll=10}, translation {z=-4}}
   n11 = "Transform (Modify) Bone - Bone: correction_hat_back"(ComponentPose=@n10, Alpha=@n9, Translation="0,0,5", Rotation="0,0,0", Scale="1,1,1") {rotationMode=BMM_Additive, translationMode=BMM_Additive, boneToModify {boneName=correction_hat_back}, translation {z=5}}
   n12 = "Transform (Modify) Bone - Bone: correction_hat_rb"(ComponentPose=@n11, Alpha=1, Translation="0.000,-8.785,2.405", Rotation="0.000,0.000,-2.305", Scale="1,1,1") {
    rotationMode=BMM_Additive
    translationMode=BMM_Additive
    alphaScaleBiasClamp {bInterpResult=true}
    boneToModify {boneName=correction_hat_rb}
    rotation {roll=-2.305}
    translation {y=-8.785, z=2.405}
   }
   Output Pose(Result=Component To Local(ComponentPose=@n12)) {layerGroup=None}
  ]}
 EventGraph {nodes [
   Event BlueprintUpdateAnimation()
   > n1 = Sequence!(execute) -> (then_0, then_1, then_2:Exec, then_3:Exec, then_4:Exec)
   > n2 = "Gravity Override State Function (Parameter Asset) (Nativized)"(
    CharacterPartAnimInstance=Self-Reference()
    GameTimeInSeconds=GetGameTimeInSeconds()
    Parameters=/BRCosmetics/Characters/Player/Female/Medium/Bodies/F_MED_VibrantShell/Meshes/F_MED_VibrantShell_Parameters
   )
   > Set Override World Gravity(Override World Gravity:Vector=@n2) -> (Output_Get:Vector)
   n3 = CheckForBackbling()
   n4 = Branch(execute=@n1.then_1, Condition=OR Boolean!(A=@n3, B=@n3.bIsBackblingACape, C:bool=@n3.bIsBackblingLong) -> (ReturnValue))
   > Set IsWearingBackbling(IsWearingBackbling:double=1) -> (Output_Get:double)
   Set IsWearingBackbling(execute=@n4.else, IsWearingBackbling=0)
   Set Is Crouching Alpha(execute=@n1.then_2, Is Crouching Alpha:double="To Float (Boolean)"(InBool=Get bIsCrouching())) -> (Output_Get:double)
   > Set ADS(ADS:double="To Float (Boolean)"(InBool=Get bIsTargeting())) -> (Output_Get:double)
   > Set HeadUpAlpha(HeadUpAlpha:double=float * float(A=GetCurveValue(CurveName=HeadUp), B=Get IsWearingBackbling()) {is=title K2Node_PromotableOperator}) -> (Output_Get:double)
   > Set UpperarmRaisedAlpha(UpperarmRaisedAlpha:double=GetCurveValue(CurveName=UpperarmRaisedR)) -> (Output_Get:double)
   > Set NewVar(NewVar:double=float * float(A=Get IsWearingBackbling(), B=Get HeadUpAlpha())) -> (Output_Get:double)
   Set torsoLeftRotationOnPickaxeAlpha(
    execute=@n1.then_3
    torsoLeftRotationOnPickaxeAlpha:float=GetAlphaFromReferenceTransformDelta(
     SkeletalMeshComponent=GetOwnerSkeletalMeshComponent()
     BoneName=spine_05
     BaseName=pelvis
     DeltaComponentType=EulerX
     InputRange=(X=0,Y=-19)
     TwistAxis="1,0,0"
    )
   ) -> (Output_Get:float)
  ]}
]
variables [
 ADS {type=number}
 HeadUpAlpha {type=number}
 Is Crouching Alpha {type=number}
 IsWearingBackbling {type=number}
 NewVar {type=number}
 Override World Gravity {type=Vector}
 RigidBody {type=boolean}
 torsoLeftRotationOnPickaxeAlpha {type=number}
 UpperarmRaisedAlpha {type=number}
]
